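# EQS query that picks a flee location for Sprites AI, away from its current target and players.
# It generates a ring of points with data-bound radius and count, projected onto navigation using the Sprites nav filter.
# Filters drop points too close to the querier or outside the bound leash radius.
# Scoring most strongly favors directions opposite the current target, then distance from all players, then distance from the querier.
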EnvQuery /Sprites/AI/EQS/EQS_Sprites_Flee_Location
options [{
  generator {
   name=EnvQueryGenerator_OnCircle_1
   class=EnvQueryGenerator_OnCircle
   properties {
    circleRadius {
     dataBinding=&self:EnvQueryGenerator_OnCircle_1.FortAIDataProvider_ScalableFloat_1
     dataField=FloatValue
     defaultValue=2000
    }
    numberOfPoints {
     dataBinding=&self:EnvQueryGenerator_OnCircle_1.FortAIDataProvider_ScalableFloat_2
     dataField=IntValue
     defaultValue=8
    }
    projectionData {
     extentX=100
     navigationFilter=/Sprites/AI/NavFilters/NavFilter_DefaultSprites.NavFilter_DefaultSprites_C
     postProjectionVerticalOffset=30
     projectDown=2000
     projectUp=2000
    }
    traceData {extentX=100, navigationFilter=&NavFilter_DefaultSprites_C}
   }
  }
  tests [
   EnvQueryTest_Distance_0 {
    class=EnvQueryTest_Distance
    properties {
     distanceTo=&EnvQueryContext_Querier
     filterType=Minimum
     floatValueMin=500
     scoringEquation=Linear
     scoringFactor=1
     testComment=Remove point to close from the Querier
     testMode=Distance3D
     testPurpose=Filter
    }
   }
   EnvQueryTest_Distance_4 {
    class=EnvQueryTest_Distance
    properties {
     clampMinType=SpecifiedValue
     distanceTo=/Script/FortniteAI.FortQueryContext_LeashCenter
     filterType=Maximum
     floatValueMin=0
     scoreClampMax=2000
     scoreClampMin=2000
     scoringEquation=InverseLinear
     scoringFactor=0.75
     testComment=Remove point outside the Leash
     testMode=Distance2D
     testPurpose=Filter
     floatValueMax {
      dataBinding=&self:EnvQueryTest_Distance_4.FortAIDataProvider_LeashComponent_0
      dataField=LeashRadius
      defaultValue=0
     }
    }
   }
   EnvQueryTest_Dot_0 {
    class=EnvQueryTest_Dot
    properties {
     clampMaxType=SpecifiedValue
     clampMinType=SpecifiedValue
     scoringEquation=InverseLinear
     testComment=Prefer oposite direction of Goal
     lineA {dirMode=TwoPoints, lineFrom=&EnvQueryContext_Querier, lineTo=&EnvQueryContext_Item}
     lineB {lineTo=&FortAthenaAIBot_EQSQueryContext_CurrentTarget}
     scoreClampMax {dataBinding=null, dataField=FloatValue, defaultValue=1}
     scoreClampMin {dataBinding=null, dataField=FloatValue, defaultValue=-1}
     scoringFactor {dataBinding=null, dataField=FloatValue, defaultValue=3}
    }
   }
   EnvQueryTest_Distance_2 {
    class=EnvQueryTest_Distance
    properties {
     distanceTo=&EnvQueryContext_Querier
     filterType=Range
     floatValueMin=0
     scoringEquation=Linear
     scoringFactor=1
     testComment=""
     testMode=Distance2D
     testPurpose=Score
    }
   }
   EnvQueryTest_Distance_1 {
    class=EnvQueryTest_Distance
    properties {
     distanceTo=/Script/FortniteAIServer.FortQueryContext_AllPlayers
     filterType=Range
     floatValueMin=0
     scoringEquation=Linear
     scoringFactor=2
     testComment=""
     testMode=Distance2D
     testPurpose=Score
    }
   }
  ]
 }]
